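# Niagara effect for the SunBurst backpack cosmetic that pops word and letter meshes upward in short bursts.
# LEAVE, ME and ALONG emitters are gated by system SpawnA/B/C flags through spawn probability, alongside an ungated emitter of random meshes.
# Two-particle words use spawn group to choose mesh and opposite sideways velocity.
# Particles tumble, follow source movement, fall under curve-scaled gravity with drag, and shrink late in life.
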NiagaraSystem /BRCosmetics/Accessories/FORT_Backpacks/Backpack_M_MED_SunBurst_Blue/FX/NS_SunBurst_Backpack_PopLetters
dependencies {
 usedDataInterfaces [&NiagaraDataInterfaceArrayFloat, &NiagaraDataInterfaceCurve]
 usedDynamicInputs [&FN_FrontEndVector3Modifier]
 usedModules [
  &self:SystemSpawnScript.SystemScriptSource.SystemScriptGraph.NiagaraNodeAssignment_0.SetVariables_#1
  &self:SystemSpawnScript.SystemScriptSource.SystemScriptGraph.NiagaraNodeAssignment_1.SetVariables_#2
  &self:SystemSpawnScript.SystemScriptSource.SystemScriptGraph.NiagaraNodeAssignment_2.SetVariables_#3
  &SystemState
  &self:SpawnLogic
 ]
}
emitters {
 PopLetters_ALONG {
  emitterUpdateScript [
   &EmitterState()
   &SpawnBurst_Instantaneous(
    Spawn Count=2
    Spawn Probability=&MakeCustomFloatFromBool(Bool=$System.SpawnC, True Float=1, False Float=0)
    Spawn Time=0.1
    Use Spawn Probability=true
    UseLoopCountLimit=false
   )
  ]
  particleSpawnScript [
   &InitializeParticle(
    Color Mode=::NewEnumerator0
    Mesh Scale Mode=::NewEnumerator3
    Mesh Uniform Scale=1.5
    Sprite Size Mode=::NewEnumerator3
    UsePositionOffset=false
    Write Material Random=true
   )
   &SystemLocation(MaskSpawnGroup=false, Offset=$System.LocationOffset)
   &FN_SetSpawnGroup(Number of Spawn Groups=2)
   &AddVelocity(Velocity=&MakeVector(
     X=&SelectFloatFromArray(Direct Array Index=$Particles.SpawnGroup, Float Selection Array=DataInterface() {propertyValues {FloatData [20,-20]}})
     Y=-20
     Z=&Multiply_Float(A=&RandomRangeFloat(Minimum=230, Maximum=250), B=1)
    )) {module=AddVelocity001}
   /SetVariables(Particles.MeshIndex=$Particles.SpawnGroup) {isSetParameters=true, module=SetVariables_#4}
  ]
  particleUpdateScript [
   &ParticleState()
   &UpdateMeshOrientation(Rotation Rate=&RandomRangeFloat(Minimum=-0.12, Maximum=0.12), Rotation Vector=<1 1 0>)
   &ScaleMeshSize(Scale Factor=&VectorFromFloat(Value=&FloatFromCurve(FloatCurve=DataInterface() {propertyValues {Curve {keys [
          {tangentMode=RCTM_User, time=0, value=1}
          {arriveTangent=-0.0405, interpMode=RCIM_Cubic, leaveTangent=-0.0405, tangentMode=RCTM_User, time=0.7009, value=1}
          {arriveTangent=-8.0089, interpMode=RCIM_Cubic, leaveTangent=-8.0089, tangentMode=RCTM_User, time=1, value=0}
         ]}}})))
   &GravityForce(Gravity=&MakeVector(
     X=0
     Y=0
     Z=&Multiply_Float(A=-980, B=&FloatFromCurve(FloatCurve=$System.GravityCurve, Scale Curve=0.6))
    ))
   &InheritSourceMovement(Applied Position Scale=<0.95 0.95 1>, ApplyForce=false, ApplyPosition=true, ApplyPositionFalloff=false)
   &Drag(Drag=2, Use Linear Drag=true, Use Rotational Drag=true)
   &SolveForcesAndVelocity(Clamp Velocity=false, Limit Acceleration=false)
  ]
  rendererClasses [&NiagaraMeshRendererProperties]
  renderers [{rendererClass=&NiagaraMeshRendererProperties, rendererIndex=0}]
 }
 PopLetters_LEAVE {
  emitterUpdateScript [
   &EmitterState()
   &SpawnBurst_Instantaneous(
    Spawn Count=2
    Spawn Probability=&MakeCustomFloatFromBool(Bool=$System.SpawnA, True Float=1, False Float=0)
    Spawn Time=0.1
    Use Spawn Probability=true
    UseLoopCountLimit=false
   )
  ]
  particleSpawnScript [
   &InitializeParticle(
    Color Mode=::NewEnumerator0
    Mesh Scale Mode=::NewEnumerator3
    Mesh Uniform Scale=1.5
    Sprite Size Mode=::NewEnumerator3
    UsePositionOffset=false
    Write Material Random=true
   )
   &SystemLocation(MaskSpawnGroup=false, Offset=$System.LocationOffset)
   &FN_SetSpawnGroup(Number of Spawn Groups=2)
   &AddVelocity(Velocity=&MakeVector(
     X=&SelectFloatFromArray(Direct Array Index=$Particles.SpawnGroup, Float Selection Array=DataInterface() {propertyValues {FloatData [20,-20]}})
     Y=-20
     Z=&Multiply_Float(A=&RandomRangeFloat(Minimum=230, Maximum=250), B=1)
    )) {module=AddVelocity001}
   /SetVariables(Particles.MeshIndex=$Particles.SpawnGroup) {isSetParameters=true, module=SetVariables_#4}
  ]
  particleUpdateScript [
   &ParticleState()
   &UpdateMeshOrientation(Rotation Rate=&RandomRangeFloat(Minimum=-0.3, Maximum=0.3), Rotation Vector=<1 1 0>)
   &ScaleMeshSize(Scale Factor=&VectorFromFloat(Value=&FloatFromCurve(FloatCurve=DataInterface() {propertyValues {Curve {keys [
          {tangentMode=RCTM_User, time=0, value=1}
          {arriveTangent=-0.0405, interpMode=RCIM_Cubic, leaveTangent=-0.0405, tangentMode=RCTM_User, time=0.7009, value=1}
          {arriveTangent=-8.0089, interpMode=RCIM_Cubic, leaveTangent=-8.0089, tangentMode=RCTM_User, time=1, value=0}
         ]}}})))
   &GravityForce(Gravity=&MakeVector(
     X=0
     Y=0
     Z=&Multiply_Float(A=-980, B=&FloatFromCurve(FloatCurve=$System.GravityCurve, Scale Curve=0.6))
    ))
   &InheritSourceMovement(Applied Position Scale=<0.95 0.95 1>, ApplyForce=false, ApplyPosition=true, ApplyPositionFalloff=false)
   &Drag(Drag=2, Use Linear Drag=true, Use Rotational Drag=true)
   &SolveForcesAndVelocity(Clamp Velocity=false, Limit Acceleration=false)
  ]
  rendererClasses [&NiagaraMeshRendererProperties]
  renderers [{rendererClass=&NiagaraMeshRendererProperties, rendererIndex=0}]
 }
 PopLetters_ME {
  emitterUpdateScript [
   &EmitterState()
   &SpawnBurst_Instantaneous(
    Spawn Probability=&MakeCustomFloatFromBool(Bool=$System.SpawnB, True Float=1, False Float=0)
    Spawn Time=0.1
    Use Spawn Probability=true
    UseLoopCountLimit=false
   )
  ]
  particleSpawnScript [
   &InitializeParticle(
    Color Mode=::NewEnumerator0
    Mesh Scale Mode=::NewEnumerator3
    Mesh Uniform Scale=1.5
    Sprite Size Mode=::NewEnumerator3
    UsePositionOffset=false
    Write Material Random=true
   )
   &AddVelocity(
    Cone Angle=5
    Cone Axis=<0 0 1>
    Module.UseConeFalloff=false
    Module.UseVelDistribution=false
    Override Seed=false
    Rotation Coordinate Space=::World
    Velocity Mode=::NewEnumerator2
    Velocity Speed=&Multiply_Float(A=&RandomRangeFloat(Minimum=200, Maximum=250), B=1)
   )
   &SystemLocation(MaskSpawnGroup=false, Offset=$System.LocationOffset)
  ]
  particleUpdateScript [
   &ParticleState()
   &UpdateMeshOrientation(Rotation Rate=&RandomRangeFloat(Minimum=-0.3, Maximum=0.3), Rotation Vector=<1 1 0>)
   &ScaleMeshSize(Scale Factor=&VectorFromFloat(Value=&FloatFromCurve(FloatCurve=DataInterface() {propertyValues {Curve {keys [
          {tangentMode=RCTM_User, time=0, value=1}
          {arriveTangent=-0.0405, interpMode=RCIM_Cubic, leaveTangent=-0.0405, tangentMode=RCTM_User, time=0.7009, value=1}
          {arriveTangent=-8.0089, interpMode=RCIM_Cubic, leaveTangent=-8.0089, tangentMode=RCTM_User, time=1, value=0}
         ]}}})))
   &GravityForce(Gravity=&MakeVector(
     X=0
     Y=0
     Z=&Multiply_Float(A=-980, B=&FloatFromCurve(FloatCurve=$System.GravityCurve, Scale Curve=0.6))
    ))
   &InheritSourceMovement(Applied Position Scale=<0.95 0.95 1>, ApplyForce=false, ApplyPosition=true, ApplyPositionFalloff=false)
   &Drag(Drag=2, Use Linear Drag=true, Use Rotational Drag=true)
   &SolveForcesAndVelocity(Clamp Velocity=false, Limit Acceleration=false)
  ]
  rendererClasses [&NiagaraMeshRendererProperties]
  renderers [{rendererClass=&NiagaraMeshRendererProperties, rendererIndex=0}]
 }
 PopLetters_Random {
  emitterSpawnScript [/SetVariables(Emitter.RandomSpawnCount:NiagaraInt32=&RandomRangeInt(Minimum=1, Maximum=3)) {isSetParameters=true}]
  emitterUpdateScript [
   &EmitterState()
   &SpawnBurst_Instantaneous(Spawn Count=$Emitter.RandomSpawnCount, Spawn Time=0.1, Use Spawn Probability=true, UseLoopCountLimit=false)
  ]
  particleSpawnScript [
   &InitializeParticle(
    Color Mode=::NewEnumerator0
    Lifetime=0.7
    Mesh Scale Mode=::NewEnumerator3
    Mesh Uniform Scale=1.5
    Sprite Size Mode=::NewEnumerator3
    UsePositionOffset=false
    Write Material Random=true
   )
   &SystemLocation(MaskSpawnGroup=false, Offset=$System.LocationOffset)
   &FN_SetSpawnGroup(Number of Spawn Groups:NiagaraInt32=$Emitter.RandomSpawnCount)
   &AddVelocity(Velocity=&MakeVector(
     X=&RandomRangeFloat(Minimum=-50, Maximum=50)
     Y=-20
     Z=&Multiply_Float(A=&RandomRangeFloat(Minimum=230, Maximum=250), B=0.6)
    )) {module=AddVelocity001}
   /SetVariables(Particles.MeshIndex=&RandomRangeInt(Minimum=0, Maximum=4)) {isSetParameters=true, module=SetVariables_#4}
   &AddVelocity(
    Cone Angle=90
    Cone Axis=<-1 0 1>
    Module.UseConeFalloff=false
    Module.UseVelDistribution=false
    Override Seed=false
    Velocity Mode=::NewEnumerator2
    Velocity Speed=&RandomRangeFloat(Minimum=20, Maximum=60)
   )
  ]
  particleUpdateScript [
   &ParticleState()
   &UpdateMeshOrientation(Rotation Rate=&RandomRangeFloat(Minimum=-0.6, Maximum=0.6), Rotation Vector=<1 1 0>)
   &ScaleMeshSize(Scale Factor=&VectorFromFloat(Value=&FloatFromCurve(FloatCurve=DataInterface() {propertyValues {Curve {keys [
          {tangentMode=RCTM_User, time=0, value=1}
          {arriveTangent=-0.0405, interpMode=RCIM_Cubic, leaveTangent=-0.0405, tangentMode=RCTM_User, time=0.7009, value=1}
          {arriveTangent=-8.0089, interpMode=RCIM_Cubic, leaveTangent=-8.0089, tangentMode=RCTM_User, time=1, value=0}
         ]}}})))
   &GravityForce(Gravity=&MakeVector(
     X=0
     Y=0
     Z=&Multiply_Float(A=-980, B=&FloatFromCurve(FloatCurve=$System.GravityCurve, Scale Curve=0.6))
    ))
   &InheritSourceMovement(Applied Position Scale=<0.95 0.95 1>, ApplyForce=false, ApplyPosition=true, ApplyPositionFalloff=false)
   &Drag(Drag=2, Use Linear Drag=true, Use Rotational Drag=true)
   &SolveForcesAndVelocity(Clamp Velocity=false, Limit Acceleration=false)
  ]
  rendererClasses [&NiagaraMeshRendererProperties]
  renderers [{rendererClass=&NiagaraMeshRendererProperties, rendererIndex=0}]
 }
}
summary {
 systemName=NS_SunBurst_Backpack_PopLetters
 emitters [
  {emitterName=PopLetters_LEAVE, rendererClasses [&NiagaraMeshRendererProperties]}
  {emitterName=PopLetters_ME, rendererClasses [&NiagaraMeshRendererProperties]}
  {emitterName=PopLetters_ALONG, rendererClasses [&NiagaraMeshRendererProperties]}
  {emitterName=PopLetters_Random, rendererClasses [&NiagaraMeshRendererProperties]}
 ]
 userVariables [
  User.bIsFrontEnd {type=NiagaraBool}
  User.bIsFrontEndPreview {type=NiagaraBool}
  User.WordNum {type=NiagaraInt32, defaultValue=1}
 ]
}
