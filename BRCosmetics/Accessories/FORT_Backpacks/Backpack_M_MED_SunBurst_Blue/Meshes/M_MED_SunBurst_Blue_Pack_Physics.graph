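# Physics asset for the SunBurst Blue backpack cosmetic mesh, defining simulated bodies for its bag and seven chip bones.
# Each bone has a single capsule body, with the bag capsule the largest.
# All seven chip bodies constrain directly to the bag bone in a star layout, not a chain.
# Tight angular limits keep most axes nearly locked, giving each chip a wider swing on mainly one axis for restrained jiggle.
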PhysicsAsset /BRCosmetics/Accessories/FORT_Backpacks/Backpack_M_MED_SunBurst_Blue/Meshes/M_MED_SunBurst_Blue_Pack_Physics
bodies {
 dyn_bag [{
   shapeType=Capsule
   length=14.5015
   radius=11.3368
   shapeName=dyn_bag_capsule
   center=<-0.6789 0.4171 -0.2409>
   rotation=<rot -5.1992 -90 10.9249>
  }]
 dyn_chips_01 [{shapeType=Capsule, length=5.7627, radius=2.4686, shapeName=dyn_chips_01_capsule, rotation=<rot -90 0 0>}]
 dyn_chips_02 [{shapeType=Capsule, length=6.0434, radius=2.5094, shapeName=dyn_chips_02_capsule}]
 dyn_chips_03 [{shapeType=Capsule, length=5.7627, radius=3.2158, shapeName=dyn_chips_03_capsule, center=<5.5816 0 0>, rotation=<rot -90 180 180>}]
 dyn_chips_04 [{shapeType=Capsule, length=4.9945, radius=3.2114, shapeName=dyn_chips_04_capsule, center=<3.8982 0 0>, rotation=<rot -90 0 0>}]
 dyn_chips_05 [{shapeType=Capsule, length=3.248, radius=3.2169, shapeName=dyn_chips_05_capsule, rotation=<rot -90 0 0>}]
 dyn_chips_06 [{shapeType=Capsule, length=0.505, radius=0.505, shapeName=dyn_chips_06_capsule, rotation=<rot -90 0 0>}]
 dyn_chips_07 [{shapeType=Capsule, length=8.8987, radius=5.5002, shapeName=dyn_chips_07_capsule, center=<0 0.1645 -0.2898>, rotation=<rot 0 -0.0046 6.9411>}]
}
constraints [
 {bone1Name=dyn_chips_07, bone2Name=dyn_bag, swing1LimitDegrees=2, swing2LimitDegrees=6, twistLimitDegrees=2}
 {bone1Name=dyn_chips_01, bone2Name=dyn_bag, swing1LimitDegrees=2, swing2LimitDegrees=25, twistLimitDegrees=2}
 {bone1Name=dyn_chips_03, bone2Name=dyn_bag, swing1LimitDegrees=2, swing2LimitDegrees=8, twistLimitDegrees=2}
 {bone1Name=dyn_chips_04, bone2Name=dyn_bag, swing1LimitDegrees=2, swing2LimitDegrees=20, twistLimitDegrees=2}
 {bone1Name=dyn_chips_05, bone2Name=dyn_bag, swing1LimitDegrees=2, swing2LimitDegrees=25, twistLimitDegrees=2}
 {bone1Name=dyn_chips_06, bone2Name=dyn_bag, swing1LimitDegrees=20, swing2LimitDegrees=2, twistLimitDegrees=2}
 {bone1Name=dyn_chips_02, bone2Name=dyn_bag, swing1LimitDegrees=2, swing2LimitDegrees=14, twistLimitDegrees=2}
]
