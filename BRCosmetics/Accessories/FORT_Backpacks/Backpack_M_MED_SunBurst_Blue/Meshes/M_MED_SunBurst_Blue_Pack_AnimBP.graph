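# Animation Blueprint for the blue SunBurst medium male backpack cosmetic, adding secondary physics motion to the pack's dynamic bones.
# The input pose runs through a RigidBody node using the pack's physics asset, simulated relative to the dyn_bag bone with overridden gravity and damped component motion.
# Three additive bone-space Modify Bone nodes then apply fixed offsets to individual dyn_chips bones before output.
# The event graph holds only unconnected default nodes, with no variables.
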AnimBlueprint /BRCosmetics/Accessories/FORT_Backpacks/Backpack_M_MED_SunBurst_Blue/Meshes/M_MED_SunBurst_Blue_Pack_AnimBP
target_skeleton=&/M_MED_SunBurst_Blue_Pack_Skeleton
graphs [
 AnimGraph {nodes [
   n1 = RigidBody(ComponentPose=Local To Component(LocalPose=Input Pose())) {
    bClampLinearTranslationLimitToRefPose=true
    bOverrideWorldGravity=true
    overridePhysicsAsset=&/M_MED_SunBurst_Blue_Pack_Physics
    simulationSpace=BaseBoneSpace
    alphaBoolBlend {blendOption=Linear}
    baseBoneRef {boneName=dyn_bag}
    componentAppliedLinearAccClamp=<3000 3000 3000>
    componentLinearAccScale=<0.75 0.75 0.75>
    componentLinearVelScale=<0.3 0.3 0.3>
    overrideWorldGravity {z=-980}
   }
   n2 = "Transform (Modify) Bone - Bone: dyn_chips_04"(ComponentPose=@n1, Alpha=1, Translation="0.276,-0.000,1.313", Rotation="-0.000,10.000,0.000", Scale="1,1,1") {
    rotationMode=BMM_Additive
    rotationSpace=BCS_BoneSpace
    translationMode=BMM_Additive
    translationSpace=BCS_BoneSpace
    alphaBoolBlend {blendOption=Linear}
    boneToModify {boneName=dyn_chips_04}
    rotation=<rot 0 10 0>
    translation {x=0.276, z=1.313}
   }
   n3 = "Transform (Modify) Bone - Bone: dyn_chips_07"(ComponentPose=@n2, Alpha=1, Translation="-0.073,0.695,-0.245", Rotation="-0.005,0.007,-10.000", Scale="1,1,1") {
    rotationMode=BMM_Additive
    rotationSpace=BCS_BoneSpace
    translationMode=BMM_Additive
    translationSpace=BCS_BoneSpace
    alphaBoolBlend {blendOption=Linear}
    boneToModify {boneName=dyn_chips_07}
    rotation=<rot -0.005 0.007 -10>
    translation=<-0.073 0.695 -0.245>
   }
   n4 = "Transform (Modify) Bone - Bone: dyn_chips_01"(ComponentPose=@n3, Alpha=1, Translation="-0.000,-0.000,0.317", Rotation="0,0,0", Scale="1,1,1") {
    rotationMode=BMM_Additive
    rotationSpace=BCS_BoneSpace
    translationMode=BMM_Additive
    translationSpace=BCS_BoneSpace
    alphaBoolBlend {blendOption=Linear}
    boneToModify {boneName=dyn_chips_01}
    translation {z=0.317}
   }
   Output Pose(Result=Component To Local(ComponentPose=@n4)) {layerGroup=None}
  ]}
 EventGraph {nodes [
   TryGetPawnOwner()
   Event BlueprintUpdateAnimation()
  ]}
]
variables []
